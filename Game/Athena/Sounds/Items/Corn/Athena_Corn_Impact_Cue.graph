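# Sound cue for corn item impacts in Athena, classed as a ranged weapon projectile effect.
# It picks one of two impact waves with equal odds, then a modulator randomly raises the pitch, on top of a raised cue-wide pitch and lowered volume.
# A corn bounce attenuation asset handles falloff, and a concurrency override caps simultaneous plays at a few, blocking new ones when full.
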SoundCue /Game/Athena/Sounds/Items/Corn/Athena_Corn_Impact_Cue
root=Modulator_5
nodes [
 {id=Modulator_5, children [Random_1], properties {pitchMax=1.3, pitchMin=1.1}}
 {id=Random_1, children [WavePlayer_6,WavePlayer_7], properties {weights [1,1]}}
 {id=WavePlayer_6, properties {soundWaveAssetPtr=&/Athena_Corn_Impact_02}}
 {id=WavePlayer_7, properties {soundWaveAssetPtr=&/Athena_Corn_Impact_01}}
]
properties {
 attenuationSettings=&/Fort_Corn_Bounce_Att
 bOverrideConcurrency=true
 maxDistance=3000
 pitchMultiplier=1.25
 soundClassObject=&Fort_WeaponRanged_Projectile_Effects
 volumeMultiplier=0.25
 concurrencyOverrides {maxCount=3, resolutionRule=PreventNew}
}
